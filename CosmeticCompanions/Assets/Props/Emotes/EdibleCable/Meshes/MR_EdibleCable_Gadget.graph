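# Modular Control Rig for the EdibleCable emote gadget prop in CosmeticCompanions, covering two darts, six balloons and a boss balloon.
# Each piece sits under a shared Root, offset and main chain and carries its own socket.
# A Root module and one module per piece expose connectors, with primary and parent control connectors on each piece.
# The rig's own graph holds only an empty Begin Execution event.
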ControlRigBlueprint /CosmeticCompanions/Assets/Props/Emotes/EdibleCable/Meshes/MR_EdibleCable_Gadget
graphs [
 RigVMFunctionLibrary {links [], nodes []}
 RigVMModel {nodes [RigUnit_BeginExecution = &RigUnit_BeginExecution()]}
]
hierarchy [
 Root {loc=<0 0 0>, rot=<0 0 0 1>}
 offset {parent=Root, loc=<0 0 0>, rot=<0 0 0 1>}
 main {parent=offset, loc=<0 0 0>, rot=<0 0 0 1>}
 Balloon01 {parent=main, loc=<0 0 0>, rot=<0 0 0 1>}
 Balloon01_socket {parent=Balloon01, type=Socket, loc=<0 0 0>, rot=<0 0 0 1>}
 dart01 {parent=main, loc=<0 0 0>, rot=<0 0 0 1>}
 dart01_socket {parent=dart01, type=Socket, loc=<0 0 0>, rot=<0 0 0 1>}
 dart02 {parent=main, loc=<0 0 0>, rot=<0 0 0 1>}
 dart02_socket {parent=dart02, type=Socket, loc=<0 0 0>, rot=<0 0 0 1>}
 Balloon05 {parent=main, loc=<0 0 0>, rot=<0 0 0 1>}
 Balloon05_socket {parent=Balloon05, type=Socket, loc=<0 0 0>, rot=<0 0 0 1>}
 Balloon04 {parent=main, loc=<0 0 0>, rot=<0 0 0 1>}
 Balloon04_socket {parent=Balloon04, type=Socket, loc=<0 0 0>, rot=<0 0 0 1>}
 Balloon03 {parent=main, loc=<0 0 0>, rot=<0 0 0 1>}
 Balloon03_socket {parent=Balloon03, type=Socket, loc=<0 0 0>, rot=<0 0 0 1>}
 Balloon02 {parent=main, loc=<0 0 0>, rot=<0 0 0 1>}
 Balloon02_socket {parent=Balloon02, type=Socket, loc=<0 0 0>, rot=<0 0 0 1>}
 Balloon06 {parent=main, loc=<0 0 0>, rot=<0 0 0 1>}
 Balloon06_socket {parent=Balloon06, type=Socket, loc=<0 0 0>, rot=<0 0 0 1>}
 BalloonBoss {parent=main, loc=<0 0 0>, rot=<0 0 0 1>}
 BalloonBoss_socket {parent=BalloonBoss, type=Socket, loc=<0 0 0>, rot=<0 0 0 1>}
 Root/Root {type=Connector, loc=<0 0 0>, rot=<0 0 0 1>}
 dart 1/Add Control Primary {type=Connector, loc=<0 0 0>, rot=<0 0 0 1>}
 dart 1/Parent Control {type=Connector, loc=<0 0 0>, rot=<0 0 0 1>}
 dart 2/Add Control Primary {type=Connector, loc=<0 0 0>, rot=<0 0 0 1>}
 dart 2/Parent Control {type=Connector, loc=<0 0 0>, rot=<0 0 0 1>}
 balloon 5/Add Control Primary {type=Connector, loc=<0 0 0>, rot=<0 0 0 1>}
 balloon 5/Parent Control {type=Connector, loc=<0 0 0>, rot=<0 0 0 1>}
 balloon 4/Add Control Primary {type=Connector, loc=<0 0 0>, rot=<0 0 0 1>}
 balloon 4/Parent Control {type=Connector, loc=<0 0 0>, rot=<0 0 0 1>}
 balloon 3/Add Control Primary {type=Connector, loc=<0 0 0>, rot=<0 0 0 1>}
 balloon 3/Parent Control {type=Connector, loc=<0 0 0>, rot=<0 0 0 1>}
 balloon 2/Add Control Primary {type=Connector, loc=<0 0 0>, rot=<0 0 0 1>}
 balloon 2/Parent Control {type=Connector, loc=<0 0 0>, rot=<0 0 0 1>}
 balloon 1/Add Control Primary {type=Connector, loc=<0 0 0>, rot=<0 0 0 1>}
 balloon 1/Parent Control {type=Connector, loc=<0 0 0>, rot=<0 0 0 1>}
 balloon 6/Add Control Primary {type=Connector, loc=<0 0 0>, rot=<0 0 0 1>}
 balloon 6/Parent Control {type=Connector, loc=<0 0 0>, rot=<0 0 0 1>}
 balloon boss/Add Control Primary {type=Connector, loc=<0 0 0>, rot=<0 0 0 1>}
 balloon boss/Parent Control {type=Connector, loc=<0 0 0>, rot=<0 0 0 1>}
]
