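# Physics asset for the EdibleCable gadget, an emote prop in CosmeticCompanions, providing simplified collision shapes for balloon and dart bones.
# Five balloon bodies each use a large, similarly sized capsule, and two dart bodies each use a thin, named capsule.
# No constraints join the seven bodies, so the asset defines no joints between them.
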PhysicsAsset /CosmeticCompanions/Assets/Props/Emotes/EdibleCable/Meshes/PA_EdibleCable_Gadget
bodies {
 Balloon01 [{shapeType=Capsule, length=23.4714, radius=18.766, center=<-0.0015 0 22.8706>, rotation=<rot 0 -90 0.201>}]
 Balloon02 [{shapeType=Capsule, length=23.4714, radius=18.766, center=<-0.0015 0 24.2612>, rotation=<rot 0 -90 0.201>}]
 Balloon03 [{shapeType=Capsule, length=23.4714, radius=18.766, center=<-0.0015 0 24.2612>, rotation=<rot 0 -90 0.201>}]
 Balloon04 [{shapeType=Capsule, length=23.4714, radius=18.766, center=<-0.0015 0 24.2612>, rotation=<rot 0 -90 0.201>}]
 Balloon05 [{shapeType=Capsule, length=23.4714, radius=18.766, center=<-0.0015 0 24.2612>, rotation=<rot 0 -90 0.201>}]
 dart01 [{shapeType=Capsule, length=10.1005, radius=1.414, shapeName=dart01_capsule, center=<0.0274 0 -0.0005>, rotation=<rot -0.0671 -90 0>}]
 dart02 [{shapeType=Capsule, length=10.1005, radius=1.414, shapeName=dart02_capsule, center=<0.0274 0 -0.0005>, rotation=<rot -0.0671 -90 0>}]
}
constraints []
